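; Creative minigame UI interface that declares functions through which game-end callouts and a HUD score option are read or overridden.
; GetGameEndCallouts takes a flag for using device-overridden callouts, while OverrideGameEndCallouts accepts victory and defeat text plus a callout type defaulting to WinLose.
; ShouldShowCumulativeScoreInHUD reports whether the HUD shows cumulative score.
; All three hold only placeholder returns, with no logic in this file.
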(fn GetGameEndCallouts
  (params (bUseDeviceOveriddenCallouts :type Boolean))
  (return "" "" "WinLose"))

(fn OverrideGameEndCallouts
  (params (VictoryCallout :type Text) (DefeatCallout :type Text) (GameEndCallout :type EMinigameGameEndCallout%20Enum :default "WinLose"))
  (return false))

(fn ShouldShowCumulativeScoreInHUD
  (return false))
